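# Niagara system for the Rapid Rush vehicle boost cosmetic, a water-styled exhaust effect at the car's boost sockets.
# A triple-layer ribbon trail with caustics joins four emitters: ripple meshes, looping shield meshes, SubUV sprite spray and expanding splash meshes.
# Particles are placed on skeletal mesh sockets across ports, shifted by a per-car offset.
# Painted-variant modules supply per-paint palettes, with bIsPainted switching secondary layers from default blue and white.
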NiagaraSystem /VehicleCosmetics/Boosts/Boost_RapidRush/FX/NS_Boost_RapidRush_V1
dependencies {
 usedDataInterfaces [
  &NiagaraDataInterfaceCurve
  &NiagaraDataInterfaceVectorCurve
  &NiagaraDataInterfaceColorCurve
  &NiagaraDataInterfaceSpriteRendererInfo
  &NiagaraDataInterfaceArrayColor
  &NiagaraDataInterfaceArrayFloat3
  &NiagaraDataInterfaceSkeletalMesh
 ]
 usedDynamicInputs []
 usedModules [
  &self:SystemSpawnScript.SystemScriptSource.SystemScriptGraph.NiagaraNodeAssignment_0.SetVariables_#1
  &SystemState
 ]
}
emitters {
 NE_VC_Boost_Ribbon_Parent {
  emitterUpdateScript [
   &EmitterState(
    Inactive Response=::NewEnumerator1
    Life Cycle Mode=::NewEnumerator1
    Loop Duration=1
    Recalculate Duration Each Loop=true
    UseLoopDelay=false
   )
   &VC_FilteredSocketCount()
   &VC_RateQualityHandler(High Spawn Rate=60, Low Spawn Rate=60, Medium Spawn Rate=60, Use Custom Rate?=true)
   &VC_SpawnRibbon_v1(Number of Ribbons=$Emitter.PortCount, SpawnRate=$Emitter.Spawn Rate)
  ]
  particleSpawnScript [
   &VC_PaintedVariantsOverride(
    Black=<rgba 0.007 0.007 0.007 1>
    BurntSienna=<rgba 0.03 0.0094 0.0037 1>
    Cobalt=<rgba 0.0085 0.0139 0.17 1>
    Crimson=<rgba 0.07 0.0105 0.0105 1>
    DefaultColor=<rgba 0 0.0385 0.3854 1>
    ForestGreen=<rgba 0 0.04 0.0067 1>
    Grey=<rgba 0.045 0.045 0.045 1>
    Lime=<rgba 0.0257 0.07 0.0001 1>
    Orange=<rgba 0.15 0.025 0 1>
    Pink=<rgba 0.14 0.0423 0.1357 1>
    Purple=<rgba 0.0463 0.0005 0.1406 1>
    R1 Black:LinearColor=<rgba 0.0237 0.0239 0.026 1.2>
    R1 BurntSienna:LinearColor=<rgba 0.17 0.0408 0.0085 1>
    R1 Cobalt:LinearColor=<rgba 0.0254 0.0356 0.6354 1>
    R1 Crimson:LinearColor=<rgba 0.2969 0.0166 0.0119 1>
    R1 DefaultColor=<rgba 0.14 0.4553 1 1>
    R1 ForestGreen:LinearColor=<rgba 0.0025 0.2396 0.0145 1>
    R1 Grey:LinearColor=<rgba 0.18 0.18 0.18 1>
    R1 Lime:LinearColor=<rgba 0.145 0.3 0 1>
    R1 Orange:LinearColor=<rgba 0.45 0.0825 0 1>
    R1 Pink:LinearColor=<rgba 0.4792 0.1437 0.4644 1>
    R1 Purple:LinearColor=<rgba 0.1783 0.0132 0.5052 1>
    R1 Saffron:LinearColor=<rgba 0.2396 0.2396 0.0176 1>
    R1 SkyBlue:LinearColor=<rgba 0.062 0.2014 0.6198 1>
    R1 TitaniumWhite:LinearColor=<rgba 0.5 0.5 0.5 1>
    R2 Black:LinearColor=<rgba 0.17 0.17 0.17 1>
    R2 BurntSienna:LinearColor=<rgba 0.7 0.312 0.2079 1>
    R2 Cobalt:LinearColor=<rgba 0.375 0.3958 1.5 1>
    R2 Crimson:LinearColor=<rgba 0.9 0.225 0.2433 1>
    R2 ForestGreen:LinearColor=<rgba 0.4 0.8 0.4202 1>
    R2 Grey:LinearColor=<rgba 1 1 1 1>
    R2 Lime:LinearColor=<rgba 0.625 1 0.25 1>
    R2 Orange:LinearColor=<rgba 1 0.4 0.2 1>
    R2 Pink:LinearColor=<rgba 0.85 0.425 0.7437 1>
    R2 Purple:LinearColor=<rgba 0.69 0.27 0.9 1>
    R2 Saffron:LinearColor=<rgba 0.95 0.95 0.2375 1>
    R2 SkyBlue:LinearColor=<rgba 0.375 0.6667 1.25 1>
    R2 TitaniumWhite:LinearColor=<rgba 1.1 1.1 1.1 1>
    Saffron=<rgba 0.0573 0.0573 0 1>
    SkyBlue=<rgba 0 0.05 0.2 1>
    TitaniumWhite=<rgba 0.1042 0.1042 0.1042 1>
    Use R1 Painted Overrides=true
    Use R2 Painted Overrides=true
   )
   &VC_InitializeRibbon_v1(
    Color=$Particles.Color
    Custom Offset=true
    Enable Skeletal Position=true
    Facing Axis:ENiagaraOrientationAxis=::ZAxis
    Lifetime=1
    Local Vector Offset=&AddVectorToPosition(Position=<0 0 0>, Vector=Linked() {linkedVariable {name=User.DTCarOffset, type=NiagaraPosition}})
    Offset Amount=1
    Ribbon Width=50
   )
   &AddVelocity(Velocity=<-1000 0 0>) {module=AddVelocity001}
  ]
  particleUpdateScript [
   &ParticleState()
   &InheritSourceMovement(
    Applied Position Scale=&VectorFromFloat(Value=&FloatFromCurve(FloatCurve=DataInterface() {propertyValues {Curve {keys [<at 0 1 RCIM_Cubic>, {arriveTangent=-2.1678, interpMode=RCIM_Cubic, leaveTangent=-2.1678, tangentMode=RCTM_User, time=1, value=0}]}}}))
    ApplyForce=false
    ApplyPosition=true
    ApplyPositionFalloff=false
   )
   &SolveForcesAndVelocity(Clamp Velocity=false, Limit Acceleration=false)
   &VC_UpdateRibbon_v1(Enable Skeletal Position=true)
   &ScaleColor(
    Scale Alpha=&FloatFromCurve(FloatCurve=DataInterface() {propertyValues {Curve {
        defaultValue=1
        keys [{interpMode=RCIM_Cubic, tangentMode=RCTM_User, time=0.1, value=0.3}, {arriveTangent=0.0749, tangentMode=RCTM_User, time=0.97, value=0}]
       }}})
    Scale RGB=&VectorFromFloat(Value=&MakeCustomFloatFromBool(
      Bool=$User.bIsPainted
      True Float=&FloatFromCurve(FloatCurve=DataInterface() {propertyValues {Curve {keys [<at 0 2 RCIM_Cubic>,<at 0.2 1 RCIM_Cubic>]}}})
      False Float=1
     ))
    ScaleA=true
    ScaleRGB=true
   )
   &ScaleRibbonWidth(Ribbon Width Scale=&FloatFromCurve(FloatCurve=DataInterface() {propertyValues {Curve {keys [<at 0 1>,<at 1 0.5>]}}}))
   /SetVariables(
    Particles.CausticsColor:LinearColor=&MakeCustomLinearColorFromBool(Bool=$User.bIsPainted, True Color=$Particles.Initial.R2 Color, False Color=<rgba 1 1 1 1>)
    Particles.CausticsWidth:NiagaraFloat=&Multiply_Float(
     A=45
     B=&FloatFromCurve(FloatCurve=DataInterface() {propertyValues {Curve {keys [<at 0 1>,<at 1 0.5>]}}})
    )
    Particles.MidTrail:LinearColor=&MakeCustomLinearColorFromBool(
     Bool=$User.bIsPainted
     True Color=&MultiplyLinearColor(
      Linear Color=$Particles.Initial.R1 Color
      Scale Factor=&MakeVector4FromVectorAndFloat(
       "Vector (RGB)"=&VectorFromFloat(Value=&FloatFromCurve(FloatCurve=DataInterface() {propertyValues {Curve {keys [<at 0 1.5 RCIM_Cubic>,<at 0.2 1 RCIM_Cubic>]}}}))
       "Float (Alpha)"=1
      )
     )
     False Color=<rgba 0.14 0.4553 1 1>
    )
    Particles.MidWidth:NiagaraFloat=&Multiply_Float(
     A=50
     B=&FloatFromCurve(FloatCurve=DataInterface() {propertyValues {Curve {keys [<at 0 1>,<at 1 0.35>]}}})
    )
   ) {isSetParameters=true}
  ]
  rendererClasses [&NiagaraRibbonRendererProperties]
  renderers [
   {rendererClass=&NiagaraRibbonRendererProperties}
   {rendererClass=&NiagaraRibbonRendererProperties, rendererIndex=1}
   {rendererClass=&NiagaraRibbonRendererProperties, rendererIndex=2}
  ]
 }
 NE_VC_Boost_Sprite_Parent001 {
  emitterUpdateScript [
   &EmitterState()
   &VC_FilteredSocketCount()
   &SpawnRate(SpawnRate=&RandomRangeFloat(Minimum=15, Maximum=10), Use Spawn Probability=true)
  ]
  particleSpawnScript [
   &VC_PaintedVariantsOverride(
    Black=<rgba 0.005 0.005 0.005 1>
    BurntSienna=<rgba 0.35 0.1006 0.0175 1>
    Cobalt=<rgba 0.015 0.0245 0.3 2>
    Crimson=<rgba 0.2 0.0177 0.0226 1>
    DefaultColor=<rgba 0 0 1 1>
    ForestGreen=<rgba 0.025 0.08 0.02 1>
    Grey=<rgba 0.07 0.07 0.07 1>
    Lime=<rgba 0.05 0.1 0 1>
    Orange=<rgba 0.5 0.2596 0.0878 1>
    Pink=<rgba 0.75 0.3164 0.5923 1>
    Purple=<rgba 0.0653 0 0.2 1>
    R1 Black:LinearColor=<rgba 0.35 0.35 0.35 1>
    R1 BurntSienna:LinearColor=<rgba 0.65 0.4117 0.325 1>
    R1 Cobalt:LinearColor=<rgba 0.45 0.4958 1 2>
    R1 Crimson:LinearColor=<rgba 0.8 0.36 0.3719 1>
    R1 ForestGreen:LinearColor=<rgba 0.2273 0.7396 0.2532 1>
    R1 Grey:LinearColor=<rgba 0.55 0.55 0.55 1>
    R1 Lime:LinearColor=<rgba 0.5118 0.6979 0.1396 1>
    R1 Orange:LinearColor=<rgba 1 0.4667 0.2 1>
    R1 Pink:LinearColor=<rgba 1.1 0.55 0.9 1>
    R1 Purple:LinearColor=<rgba 0.658 0.3 1 1>
    R1 Saffron:LinearColor=<rgba 0.7604 0.7192 0.2661 1>
    R1 SkyBlue:LinearColor=<rgba 0.4 0.65 1 1>
    R1 TitaniumWhite:LinearColor=<rgba 0.7 0.7 0.7 1>
    R2 Black:LinearColor=<rgba 0.28 0.28 0.28 1>
    R2 BurntSienna:LinearColor=<rgba 0.35 0.1416 0.0857 1>
    R2 Cobalt:LinearColor=<rgba 0.6 0.645 1.5 2>
    R2 Crimson:LinearColor=<rgba 1.15 0.46 0.4787 1>
    R2 ForestGreen:LinearColor=<rgba 0.5417 1 0.5 1>
    R2 Grey:LinearColor=<rgba 0.85 0.85 0.85 1>
    R2 Lime:LinearColor=<rgba 0.8377 1.1 0.385 1>
    R2 Orange:LinearColor=<rgba 0.7 0.2736 0.105 1>
    R2 Pink:LinearColor=<rgba 0.6 0.3 0.55 1>
    R2 Purple:LinearColor=<rgba 0.96 0.48 1.2 1>
    R2 Saffron:LinearColor=<rgba 1.5 1.3913 0.45 1>
    R2 SkyBlue:LinearColor=<rgba 0.6831 0.8776 1.15 1>
    R2 TitaniumWhite:LinearColor=<rgba 1.2 1.2 1.2 1>
    Saffron=<rgba 0.1 0.1 0.025 1>
    SkyBlue=<rgba 0 0.0823 0.247 1>
    TitaniumWhite=<rgba 0.1357 0.1357 0.1357 1>
    Use R1 Painted Overrides=true
    Use R2 Painted Overrides=true
   )
   &InitializeParticle(
    Color=$Particles.Color
    Lifetime Max=1
    Lifetime Mode=::NewEnumerator1
    Mesh Scale Mode=::NewEnumerator3
    Mesh Uniform Scale=0.5
    Sprite Rotation Mode=::NewEnumerator2
    UsePositionOffset=false
    Write Material Random=true
   )
   &PartitionParticles(Number Of Partitions=$Emitter.PortCount)
   &SkeletalMeshLocation(
    Bone / Socket Index=$Particles.PartitionParticles.Partition
    Bone / Socket Sampling Mode=::NewEnumerator5
    MaskSpawnGroup=false
    Mesh Sampling Type=::NewEnumerator6
    Sampled Position Offset=&AddVectorToPosition(Position=<-60 0 0>, Vector=Linked() {linkedVariable {name=User.DTCarOffset, type=NiagaraPosition}})
    Skeletal Mesh=$System.SkeletalMesh
   )
   &AddVelocity(Velocity=<-1500 0 0>) {enabled=false}
   &InitialMeshOrientation(
    Mesh Orientation Mode=::NewEnumerator2
    Rotation=&RandomRangeVector(Minimum=<-1 0 0>, Maximum=<1 0 0>)
    Use Rotation Vector=true
   )
  ]
  particleUpdateScript [
   &ParticleState()
   &ScaleMeshSize(Scale Factor=&VectorFromCurve(
     VectorCurve=DataInterface() {propertyValues {
       XCurve {
        defaultValue=0.25
        keys [
         {arriveTangent=1.859, interpMode=RCIM_Cubic, leaveTangent=1.859, tangentMode=RCTM_User, time=0, value=0}
         {arriveTangent=-2.5165, interpMode=RCIM_Cubic, leaveTangent=-2.5165, tangentMode=RCTM_User, time=2, value=0}
        ]
       }
       YCurve {
        defaultValue=0.25
        keys [
         {arriveTangent=1.859, interpMode=RCIM_Cubic, leaveTangent=1.859, tangentMode=RCTM_User, time=0, value=0}
         {arriveTangent=-2.5165, interpMode=RCIM_Cubic, leaveTangent=-2.5165, tangentMode=RCTM_User, time=2, value=0}
        ]
       }
       ZCurve {
        defaultValue=0.25
        keys [
         {arriveTangent=1.859, interpMode=RCIM_Cubic, leaveTangent=1.859, tangentMode=RCTM_User, time=0, value=0}
         {arriveTangent=-2.5165, interpMode=RCIM_Cubic, leaveTangent=-2.5165, tangentMode=RCTM_User, time=2, value=0}
        ]
       }
      }}
     Scale Curve=&VectorFromFloat(Value=1)
    ))
   /SetVariables(
    Particles.LightBlueRipple:LinearColor=&MakeCustomLinearColorFromBool(
     Bool=$User.bIsPainted
     True Color=$Particles.Initial.R1 Color
     False Color=&ColorFromCurve(ColorCurve=DataInterface() {propertyValues {
        AlphaCurve {keys [<at 0.75 1 RCIM_Cubic>,<at 1 0 RCIM_Cubic>]}
        BlueCurve {keys [<at 0 1 RCIM_Cubic>,<at 0.7781 1 RCIM_Cubic>]}
        GreenCurve {keys [<at 0 0.7715 RCIM_Cubic>,<at 0.7781 0.8252 RCIM_Cubic>]}
        RedCurve {keys [<at 0 0.15 RCIM_Cubic>,<at 0.7781 0.35 RCIM_Cubic>]}
       }})
    )
    Particles.LightBlueScale:Vector3f=&RandomRangeVector(
     Minimum=&VectorFromCurve(
      VectorCurve=DataInterface() {propertyValues {XCurve {keys [<at 0 0>,<at 0.75 1>]}, YCurve {keys [<at 0 0>,<at 0.75 1>]}, ZCurve {keys [<at 0 0>,<at 0.75 1>]}}}
      Scale Curve=&VectorFromFloat(Value=0.7)
     )
     Maximum=&VectorFromCurve(
      VectorCurve=DataInterface() {propertyValues {XCurve {keys [<at 0 0>,<at 1 1>]}, YCurve {keys [<at 0 0>,<at 1 1>]}, ZCurve {keys [<at 0 0>,<at 1 1>]}}}
      Scale Curve=&VectorFromFloat(Value=0.8)
     )
    )
    Particles.WhiteRipleScale:Vector3f=&RandomRangeVector(
     Minimum=&VectorFromCurve(
      VectorCurve=DataInterface() {propertyValues {XCurve {keys [<at 0 0.25>,<at 1 1>]}, YCurve {keys [<at 0 0.25>,<at 1 1>]}, ZCurve {keys [<at 0 0.25>,<at 1 1>]}}}
      Scale Curve=&VectorFromFloat(Value=0.7)
     )
     Maximum=&VectorFromCurve(
      VectorCurve=DataInterface() {propertyValues {XCurve {keys [<at 0 0>,<at 2 1>]}, YCurve {keys [<at 0 0>,<at 2 1>]}, ZCurve {keys [<at 0 0>,<at 2 1>]}}}
      Scale Curve=&VectorFromFloat(Value=0.8)
     )
    )
    Particles.WhiteRipple:LinearColor=&MakeCustomLinearColorFromBool(Bool=$User.bIsPainted, True Color=$Particles.Initial.R2 Color, False Color=<rgba 1 1 1 1>)
   ) {isSetParameters=true}
   &UpdateMeshOrientation(Rotation Rate=&RandomRangeFloat(Minimum=-0.5, Maximum=1)) {enabled=false}
   &SolveForcesAndVelocity(Clamp Velocity=false, Limit Acceleration=false)
  ]
  rendererClasses [&NiagaraMeshRendererProperties, &NiagaraSpriteRendererProperties]
  renderers [
   {rendererClass=&NiagaraMeshRendererProperties, rendererIndex=0}
   {rendererClass=&NiagaraMeshRendererProperties, rendererIndex=1}
   {rendererClass=&NiagaraMeshRendererProperties, rendererIndex=2}
   {rendererClass=&NiagaraSpriteRendererProperties, rendererIndex=3}
  ]
 }
 NE_VC_Boost_Sprite_Parent002 {
  emitterUpdateScript [
   &EmitterState(
    Inactive Response=::NewEnumerator1
    Life Cycle Mode=::NewEnumerator1
    Loop Behavior=::NewEnumerator1
    Loop Duration Mode=::NewEnumerator1
    UseLoopDelay=false
   )
   &SpawnBurst_Instantaneous(Spawn Count=4, Use Spawn Probability=false, UseLoopCountLimit=false) {module=SpawnBurst_Instantaneous001}
   &SpawnBurst_Instantaneous(Spawn Count=4, Spawn Group=1, Use Spawn Probability=false, UseLoopCountLimit=false)
   &VC_FilteredSocketCount()
   &SpawnRate(SpawnRate=&RandomRangeFloat(Minimum=20, Maximum=25), Use Spawn Probability=false) {enabled=false}
  ]
  particleSpawnScript [
   &VC_PaintedVariantsOverride(
    Black=<rgba 0.007 0.007 0.007 1>
    BurntSienna=<rgba 0.15 0.03 0 1>
    Cobalt=<rgba 0.044 0.0524 0.55 2>
    Crimson=<rgba 0.4 0.02 0.0263 1>
    DefaultColor=<rgba 0.4 0.9333 2 0.5>
    ForestGreen=<rgba 0.0625 0.25 0.025 1>
    Grey=<rgba 0.2 0.2 0.2 1>
    Lime=<rgba 0.156 0.35 0.0175 1>
    Orange=<rgba 0.6 0.11 0 1>
    Pink=<rgba 0.6 0.15 0.4875 1>
    Purple=<rgba 0.2917 0 0.7 1>
    Saffron=<rgba 0.3 0.2763 0.015 1>
    SkyBlue=<rgba 0.065 0.26 0.65 1>
    TitaniumWhite=<rgba 0.5 0.5 0.5 1>
   )
   &InitializeParticle(
    Color=$Particles.Color
    Lifetime=2
    Mesh Scale Mode=::NewEnumerator3
    Mesh Uniform Scale=&MergeHelper_FloatSwitch(0=1, 1=0.5, 2=?, 3=?, Index=&SpawnGroup(), Num=::NewEnumerator0)
    Sprite Rotation Mode=::NewEnumerator2
    UsePositionOffset=false
    Write Material Random=true
   )
   &PartitionParticles(Number Of Partitions=$Emitter.PortCount)
   &SkeletalMeshLocation(
    Bone / Socket Index=$Particles.PartitionParticles.Partition
    Bone / Socket Sampling Mode=::NewEnumerator5
    MaskSpawnGroup=false
    Mesh Sampling Type=::NewEnumerator6
    Sampled Position Offset=&AddVectorToPosition(
     Position=&MergeHelper_VectorSwitch(0=<-55 0 0>, 1=<-54 0 0>, 2=?, 3=?, Index=&SpawnGroup(), Num=::NewEnumerator0)
     Vector=Linked() {linkedVariable {name=User.DTCarOffset, type=NiagaraPosition}}
    )
    Skeletal Mesh=$System.SkeletalMesh
   )
   &AddVelocity(Velocity=<-1500 0 0>) {enabled=false}
   /SetVariables(Particles.VisibilityTag=$Emitter.SpawnGroup) {isSetParameters=true}
  ]
  particleUpdateScript [
   &ParticleState(Kill Particles When Lifetime Has Elapsed=false, Loop Particles Lifetime=true)
   &ScaleColor(Scale RGB=&VectorFromFloat(Value=1.5), ScaleA=true, ScaleRGB=true)
   &SolveForcesAndVelocity(Clamp Velocity=false, Limit Acceleration=false)
   /SetVariables(Particles.Shield:LinearColor=&MultiplyLinearColor(
     Linear Color=$Particles.Color
     Scale Factor=&MakeVector4FromVectorAndFloat("Vector (RGB)"=&VectorFromFloat(Value=2), "Float (Alpha)"=1)
    )) {isSetParameters=true}
  ]
  rendererClasses [&NiagaraMeshRendererProperties, &NiagaraSpriteRendererProperties]
  renderers [
   {rendererClass=&NiagaraMeshRendererProperties, rendererIndex=0}
   {rendererClass=&NiagaraSpriteRendererProperties, rendererIndex=1}
  ]
 }
 NE_VC_Boost_Sprite_Parent003 {
  emitterUpdateScript [
   &EmitterState()
   &SpawnBurst_Instantaneous(Spawn Count=15, Use Spawn Probability=false, UseLoopCountLimit=false)
   &VC_FilteredSocketCount()
   &SpawnRate(SpawnRate=&RandomRangeFloat(Minimum=75, Maximum=120), Use Spawn Probability=false)
  ]
  particleSpawnScript [
   &VC_PaintedVariantsOverride(
    Black=<rgba 0.4 0.4 0.4 1>
    BurntSienna=<rgba 0.5365 0.1399 0.0335 1>
    Cobalt=<rgba 0.1336 0.1459 0.7448 2>
    Crimson=<rgba 0.7865 0.2622 0.2764 1>
    ForestGreen=<rgba 0.2097 0.6823 0.2335 1>
    Grey=<rgba 0.3802 0.3802 0.3802 1>
    Lime=<rgba 0.3646 0.7292 0 1>
    Orange=<rgba 0.7917 0.2761 0 1>
    Pink=<rgba 0.974 0.2232 0.701 1>
    Purple=<rgba 0.7124 0.1641 1.5 1>
    Saffron=<rgba 0.3755 0.4427 0 1>
    SkyBlue=<rgba 0.1697 0.4332 0.9583 1>
   )
   &InitializeParticle(
    Color=$Particles.Color
    Lifetime=&RandomRangeFloat(Minimum=0.25, Maximum=0.42)
    Sprite Rotation Mode=::NewEnumerator2
    Sprite Size Mode=::NewEnumerator1
    Uniform Sprite Size Max=15
    Uniform Sprite Size Min=10
    UsePositionOffset=false
    Write Material Random=true
   )
   &PartitionParticles(Number Of Partitions=$Emitter.PortCount)
   &SkeletalMeshLocation(
    Bone / Socket Index=$Particles.PartitionParticles.Partition
    Bone / Socket Sampling Mode=::NewEnumerator5
    MaskSpawnGroup=false
    Mesh Sampling Type=::NewEnumerator6
    Sampled Position Offset=&AddVectorToPosition(Position=<-75 0 0>, Vector=Linked() {linkedVariable {name=User.DTCarOffset, type=NiagaraPosition}})
    Skeletal Mesh=$System.SkeletalMesh
   )
   &AddVelocity(
    Cone Angle=55
    Inner Cone Angle=25
    Module.UseConeFalloff=false
    Module.UseVelDistribution=false
    Override Seed=false
    Velocity Mode=::NewEnumerator2
    Velocity Speed=-250
   )
   &ShapeLocation(MaskSpawnGroup=false, Override Seed=false, Sphere Radius=25, UseOwnerScale=true)
  ]
  particleUpdateScript [
   &ParticleState()
   &ScaleSpriteSize(Uniform Curve Sprite Scale=DataInterface() {propertyValues {Curve {defaultValue=0, keys [{interpMode=RCIM_Cubic, tangentMode=RCTM_User, time=0.5, value=1}, {tangentMode=RCTM_User, time=1, value=0}]}}})
   &SubUVAnimation(
    "Sprite Renderer "=DataInterface() {propertyValues {SpriteRenderer=&self:NE_VC_Boost_Sprite_Parent003_0.NiagaraSpriteRendererProperties_1}}
    UseEndFrame=false
    UseStartFrame=false
   )
   &ScaleColor(ScaleA=true, ScaleRGB=true)
   &SolveForcesAndVelocity(Clamp Velocity=false, Limit Acceleration=false)
  ]
  rendererClasses [&NiagaraSpriteRendererProperties]
  renderers [{rendererClass=&NiagaraSpriteRendererProperties}]
 }
 NE_VC_Boost_Sprite_Parent005 {
  emitterUpdateScript [
   &EmitterState(Life Cycle Mode=::NewEnumerator1, Loop Duration=0.4, UseLoopDelay=false)
   &VC_FilteredSocketCount()
   &SpawnRate(SpawnRate=12, Use Spawn Probability=false)
  ]
  particleSpawnScript [
   &VC_PaintedVariantsOverride(
    Black=<rgba 0.02 0.02 0.02 1>
    BurntSienna=<rgba 0.2 0.0539 0.0247 1>
    Cobalt=<rgba 0.04 0.0553 0.5 1>
    Crimson=<rgba 0.4 0.02 0.0303 1>
    DefaultAlpha:NiagaraFloat=&FloatFromCurve(FloatCurve=DataInterface() {propertyValues {Curve {defaultValue=1, keys [{interpMode=RCIM_Cubic, tangentMode=RCTM_User, time=0.5, value=1}, {tangentMode=RCTM_User, time=1, value=0}]}}})
    DefaultColor=&SelectLinearColorFromArray(
     Array Sampling Mode=::NewEnumerator0
     Direct Array Index=?
     Color Selection Array=DataInterface() {propertyValues {ColorData [<rgba 0 0.1233 0.5 1.5>, <rgba 0.0563 0.149 0.4323 1.5>]}}
     Random Seed=0
    )
    ForestGreen=<rgba 0 0.2 0.0101 1>
    Grey=<rgba 0.17 0.17 0.17 1>
    Lime=<rgba 0.0933 0.2 0 1>
    Orange=<rgba 0.4 0.06 0 1>
    Pink=<rgba 0.5 0.15 0.4417 1>
    Purple=<rgba 0.1667 0 0.5 1>
    R1 Black:LinearColor=<rgba 0.2 0.2 0.2 1>
    R1 BurntSienna:LinearColor=<rgba 0.5 0.224 0.15 1>
    R1 Cobalt:LinearColor=<rgba 0.375 0.4479 1.25 1>
    R1 Crimson:LinearColor=<rgba 1 0.3 0.319 1>
    R1 ForestGreen:LinearColor=<rgba 0.4 1 0.4303 1>
    R1 Grey:LinearColor=<rgba 0.6 0.6 0.6 1>
    R1 Lime:LinearColor=<rgba 0.6667 1 0.2 1>
    R1 Orange:LinearColor=<rgba 1 0.4 0.2 1>
    R1 Pink:LinearColor=<rgba 1 0.4 0.9 1>
    R1 Purple:LinearColor=<rgba 0.6462 0.4 1 1>
    R1 Saffron:LinearColor=<rgba 1 0.925 0.1 1>
    R1 SkyBlue:LinearColor=<rgba 0.4 0.7992 1 1>
    R1 TitaniumWhite:LinearColor=<rgba 1 1 1 1>
    Saffron=<rgba 0.25 0.25 0 1>
    SkyBlue=<rgba 0 0.1788 0.5365 1>
    TitaniumWhite=<rgba 0.27 0.27 0.27 1>
    Use R1 Painted Overrides=true
   )
   &InitializeParticle(
    Color=$Particles.Color
    Lifetime=0.6
    Mesh Scale Mode=::NewEnumerator4
    Sprite Rotation Mode=::NewEnumerator2
    Sprite Size Mode=::NewEnumerator3
    UsePositionOffset=false
    Write Material Random=true
   )
   &PartitionParticles(Number Of Partitions=$Emitter.PortCount)
   &SkeletalMeshLocation(
    Bone / Socket Index=$Particles.PartitionParticles.Partition
    Bone / Socket Sampling Mode=::NewEnumerator5
    MaskSpawnGroup=false
    Mesh Sampling Type=::NewEnumerator6
    Sampled Position Offset=&AddVectorToPosition(Position=<-63 0 0>, Vector=Linked() {linkedVariable {name=User.DTCarOffset, type=NiagaraPosition}})
    Skeletal Mesh=$System.SkeletalMesh
   )
   &AddVelocity(Velocity=<-1500 0 0>) {enabled=false}
   &InitialMeshOrientation(
    Mesh Orientation Mode=::NewEnumerator2
    Orientation Vector=<-1 0 0>
    Rotation=&SelectVectorFromArray(
     Array Sampling Mode=::NewEnumerator0
     Direct Array Index=?
     Vector Selection Array=DataInterface() {propertyValues {FloatData [<0.2 0 0>, <1 0 0>, <0.4 0 0>]}}
     Random Seed=0
    )
    Use Rotation Vector=true
   )
  ]
  particleUpdateScript [
   &ParticleState()
   &SolveForcesAndVelocity(Clamp Velocity=false, Limit Acceleration=false)
   /SetVariables(
    Particles.SplashScale:Vector3f=&VectorFromCurve(
     VectorCurve=DataInterface() {propertyValues {
       XCurve {defaultValue=0.8, keys [<at 0.1 1>, <at 0.3 1.5>, {tangentMode=RCTM_User, time=0.7, value=1.7}]}
       YCurve {defaultValue=0.8, keys [<at 0.1 1>, <at 0.3 1.5>, {tangentMode=RCTM_User, time=0.7, value=1.7}]}
       ZCurve {defaultValue=0.8, keys [<at 0.1 1>, <at 0.3 1.5>, {tangentMode=RCTM_User, time=0.7, value=1.7}]}
      }}
     Scale Curve=<0.35 0.8 0.8>
    )
    Particles.TipScale:Vector3f=&VectorFromCurve(
     VectorCurve=DataInterface() {propertyValues {
       XCurve {defaultValue=0.8, keys [<at 0.1 1>, <at 0.3 1.5>, {tangentMode=RCTM_User, time=0.7, value=1.7}]}
       YCurve {defaultValue=0.8, keys [<at 0.1 1>, <at 0.3 1.5>, {tangentMode=RCTM_User, time=0.7, value=1.7}]}
       ZCurve {defaultValue=0.8, keys [<at 0.1 1>, <at 0.3 1.5>, {tangentMode=RCTM_User, time=0.7, value=1.7}]}
      }}
     Scale Curve=<0.35 0.9 0.9>
    )
    Particles.TipsColor:LinearColor=&MakeCustomLinearColorFromBool(
     Bool=$User.bIsPainted
     True Color=$Particles.Initial.R1 Color
     False Color=&ColorFromCurve(ColorCurve=DataInterface() {propertyValues {
        AlphaCurve {keys [<at 0.5081 1 RCIM_Cubic>,<at 1 0 RCIM_Cubic>]}
        BlueCurve {keys [<at 0 1 RCIM_Cubic>,<at 1 1 RCIM_Cubic>]}
        GreenCurve {keys [<at 0 1 RCIM_Cubic>,<at 1 1 RCIM_Cubic>]}
        RedCurve {keys [<at 0 1 RCIM_Cubic>,<at 1 1 RCIM_Cubic>]}
       }})
    )
   ) {isSetParameters=true}
   &ScaleColor(
    Scale Alpha=&FloatFromCurve(
     FloatCurve=DataInterface() {propertyValues {Curve {defaultValue=1, keys [{tangentMode=RCTM_User, time=0, value=0}, <at 0.7037 0.7037>, {tangentMode=RCTM_User, time=1, value=0}]}}}
     Scale Curve=&MakeCustomFloatFromBool(Bool=$User.bIsPainted, True Float=2, False Float=1)
    )
    ScaleA=true
    ScaleRGB=true
   )
   &DynamicMaterialParameters(
    Index 0 Param 2=1.9
    Index 0 Param 3=&FloatFromCurve(
     FloatCurve=DataInterface() {propertyValues {Curve {
        defaultValue=1
        keys [
         {arriveTangent=3.8224, interpMode=RCIM_Cubic, leaveTangent=3.8224, tangentMode=RCTM_User, time=0, value=0}
         {arriveTangentWeight=0.6667, interpMode=RCIM_Cubic, tangentMode=RCTM_User, time=2, value=6}
        ]
       }}}
     Scale Curve=0.8
    )
    Index 0 Param 4=1
    Module.Index0Float0Write=true
    Module.Index0Float1Write=true
    Module.Index0Float2Write=true
    Module.Index0Float3Write=true
   )
  ]
  rendererClasses [&NiagaraMeshRendererProperties, &NiagaraSpriteRendererProperties]
  renderers [
   {rendererClass=&NiagaraMeshRendererProperties, rendererIndex=0}
   {rendererClass=&NiagaraMeshRendererProperties, rendererIndex=1}
   {rendererClass=&NiagaraSpriteRendererProperties, rendererIndex=2}
  ]
 }
}
summary {
 systemName=NS_Boost_RapidRush_V1
 emitters [
  {emitterName=NE_VC_Boost_Ribbon_Parent, rendererClasses [&NiagaraRibbonRendererProperties]}
  {
   emitterName=NE_VC_Boost_Sprite_Parent001
   rendererClasses [&NiagaraMeshRendererProperties, &NiagaraSpriteRendererProperties]
  }
  {
   emitterName=NE_VC_Boost_Sprite_Parent002
   rendererClasses [&NiagaraMeshRendererProperties, &NiagaraSpriteRendererProperties]
  }
  {emitterName=NE_VC_Boost_Sprite_Parent003, rendererClasses [&NiagaraSpriteRendererProperties]}
  {
   emitterName=NE_VC_Boost_Sprite_Parent005
   rendererClasses [&NiagaraMeshRendererProperties, &NiagaraSpriteRendererProperties]
  }
 ]
 userVariables [
  User.bIsFrontendPreview {type=NiagaraBool}
  User.bIsPainted {type=NiagaraBool}
  User.ColorVariantID {type=NiagaraInt32}
  User.DTCarOffset {type=NiagaraPosition, description="", defaultValue {}}
  User.PaintedColor {type=LinearColor, description="", defaultValue {struct=LinearColor, value=<rgba 1 1 1 1>}}
  User.PaintedScalar {type=NiagaraFloat, defaultValue=1, description=""}
  User.PortCount {type=NiagaraInt32, defaultValue=2}
 ]
}
